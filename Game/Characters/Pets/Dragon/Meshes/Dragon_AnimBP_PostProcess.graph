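# Post-process AnimBP for the Dragon pet on the Chameleon skeleton, adding secondary physics motion to the cached input pose.
# When bPetWantsAnimDynamics is set, it uses a RigidBody sim during emote or DBNO, otherwise speed-weighted AnimDynamics on spine_03, neck_02 and head.
# When unset, the input pose passes through that stage.
# AnimDynamics on three fin bones and the anklet always follows, with all physics nodes LOD-limited.
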AnimBlueprint /Game/Characters/Pets/Dragon/Meshes/Dragon_AnimBP_PostProcess
target_skeleton=/Game/Characters/Pets/Chameleon/Meshes/Chameleon_Skeleton
graphs [
 AnimGraph {nodes [
   n1 = Get PawnSpeedForAnimDynamics() -> (PawnSpeedForAnimDynamics:float)
   n2 = Anim Dynamics - Bone: spine_03(ComponentPose=Local To Component(LocalPose=Use cached pose 'Sub Graph Cache'()), Alpha=@n1, ExternalForce="0,0,0") {
    angularDampingOverride=0.999
    angularSpringConstant=200
    bGravityOverrideInSimSpace=true
    bOverrideAngularDamping=true
    gravityScale=0.35
    lODThreshold=1
    numSolverIterationsPostUpdate=4
    numSolverIterationsPreUpdate=16
    alphaBoolBlend {blendOption=Linear}
    alphaScaleBiasClamp {bInterpResult=true, bMapRange=true, inRange {max=300}}
    boundBone {boneName=spine_03}
    componentLinearAccScale=<0.075 0.075 0.075>
    physicsBodyDefinitions [{
      sphereCollisionRadius=0
      boundBone {boneName=spine_03}
      boxExtents=<3 10 10>
      constraintSetup {angularTargetAxis=AxisY, twistAxis=AxisY, angularLimitsMax=<10 3 5>, angularLimitsMin=<-10 -3 -5>, angularTarget=<0 0 0>}
      localJointOffset=<0 2 0>
     }]
   }
   n3 = Anim Dynamics - Bone: neck_02(ComponentPose=@n2, Alpha=@n1, ExternalForce="0,0,0") {
    angularDampingOverride=0.999
    angularSpringConstant=200
    bGravityOverrideInSimSpace=true
    bOverrideAngularDamping=true
    gravityScale=0.35
    lODThreshold=1
    numSolverIterationsPostUpdate=4
    numSolverIterationsPreUpdate=16
    alphaBoolBlend {blendOption=Linear}
    alphaScaleBiasClamp {bInterpResult=true, bMapRange=true, inRange {max=300}}
    boundBone {boneName=neck_02}
    componentLinearAccScale=<0.075 0.075 0.075>
    physicsBodyDefinitions [{
      sphereCollisionRadius=0
      boundBone {boneName=neck_02}
      boxExtents=<3 10 10>
      constraintSetup {angularTargetAxis=AxisY, twistAxis=AxisY, angularLimitsMax=<10 3 5>, angularLimitsMin=<-10 -3 -5>, angularTarget=<0 0 0>}
      localJointOffset=<0 2 0>
     }]
   }
   n4 = Anim Dynamics - Bone: head(ComponentPose=@n3, Alpha=@n1, ExternalForce="0,0,0") {
    angularDampingOverride=0.999
    angularSpringConstant=200
    bGravityOverrideInSimSpace=true
    bOverrideAngularDamping=true
    gravityScale=0.35
    lODThreshold=1
    numSolverIterationsPostUpdate=4
    numSolverIterationsPreUpdate=16
    alphaBoolBlend {blendOption=Linear}
    alphaScaleBiasClamp {bInterpResult=true, bMapRange=true, inRange {max=300}}
    boundBone {boneName=head}
    componentLinearAccScale=<0.075 0.075 0.075>
    physicsBodyDefinitions [{
      sphereCollisionRadius=0
      boundBone {boneName=head}
      boxExtents=<3 10 10>
      constraintSetup {angularTargetAxis=AxisY, twistAxis=AxisY, angularLimitsMax=<10 3 5>, angularLimitsMin=<-10 -3 -5>, angularTarget=<0 0 0>}
      localJointOffset=<0 2 0>
     }]
   }
   n5 = RigidBody(ComponentPose=Local To Component(LocalPose=Use cached pose 'Sub Graph Cache'())) {
    bOverrideWorldGravity=true
    lODThreshold=1
    simulationSpace=BaseBoneSpace
    alphaBoolBlend {blendOption=Linear}
    baseBoneRef {boneName=root}
    componentLinearAccScale=<1 1 1>
    overrideWorldGravity {z=-980}
   }
   n6 = Blend Poses by bool(
    BlendPose_0=Component To Local(ComponentPose=@n5)
    BlendPose_1=Component To Local(ComponentPose=@n4)
    BlendTime_0=1
    BlendTime_1=1
    bActiveValue=Get bIsPlayingEmoteOrDBNO() -> (bIsPlayingEmoteOrDBNO:bool)
   ) {blendType=Linear}
   n7 = Use cached pose 'Sub Graph Cache'()
   n8 = Get bPetWantsAnimDynamics() -> (bPetWantsAnimDynamics:bool)
   n9 = Anim Dynamics - Bone: fin_btm(
    ComponentPose=Local To Component(LocalPose=Blend Poses by bool(BlendPose_0=@n6, BlendPose_1=@n7, BlendTime_0=1, BlendTime_1=1, bActiveValue=@n8) {blendType=Linear})
    Alpha=0.75
    ExternalForce="0,0,0"
   ) {
    angularDampingOverride=0.95
    angularSpringConstant=1000
    bAngularSpring=true
    bGravityOverrideInSimSpace=true
    bOverrideAngularDamping=true
    bUsePlanarLimit=false
    gravityScale=0.75
    lODThreshold=1
    numSolverIterationsPostUpdate=2
    numSolverIterationsPreUpdate=8
    simulationSpace=Actor
    alphaBoolBlend {blendOption=Linear}
    boundBone {boneName=fin_btm}
    physicsBodyDefinitions [{
      sphereCollisionRadius=0
      boundBone {boneName=fin_btm}
      boxExtents=<5 5 2.5>
      constraintSetup {angularTargetAxis=AxisZ, twistAxis=AxisZ, angularLimitsMax=<15 3 0>, angularLimitsMin=<-15 -3 0>, angularTarget=<0 0 0>}
      localJointOffset=<0 0 -0.05>
     }]
   }
   n10 = Anim Dynamics - Bone: fin_mid(ComponentPose=@n9, Alpha=0.75, ExternalForce="0,0,0") {
    angularDampingOverride=0.95
    angularSpringConstant=1000
    bAngularSpring=true
    bGravityOverrideInSimSpace=true
    bOverrideAngularDamping=true
    bUsePlanarLimit=false
    gravityScale=0.75
    lODThreshold=1
    numSolverIterationsPostUpdate=2
    numSolverIterationsPreUpdate=8
    simulationSpace=Actor
    alphaBoolBlend {blendOption=Linear}
    boundBone {boneName=fin_mid}
    physicsBodyDefinitions [{
      sphereCollisionRadius=0
      boundBone {boneName=fin_mid}
      boxExtents=<5 5 2.5>
      constraintSetup {angularTargetAxis=AxisZ, twistAxis=AxisZ, angularLimitsMax=<15 3 0>, angularLimitsMin=<-15 -3 0>, angularTarget=<0 0 0>}
      localJointOffset=<0 0 -0.05>
     }]
   }
   n11 = Anim Dynamics - Bone: fin_top(ComponentPose=@n10, Alpha=0.75, ExternalForce="0,0,0") {
    angularDampingOverride=0.95
    angularSpringConstant=1000
    bAngularSpring=true
    bGravityOverrideInSimSpace=true
    bOverrideAngularDamping=true
    bUsePlanarLimit=false
    gravityScale=0.75
    lODThreshold=1
    numSolverIterationsPostUpdate=2
    numSolverIterationsPreUpdate=8
    simulationSpace=Actor
    alphaBoolBlend {blendOption=Linear}
    boundBone {boneName=fin_top}
    physicsBodyDefinitions [{
      sphereCollisionRadius=0
      boundBone {boneName=fin_top}
      boxExtents=<5 5 2.5>
      constraintSetup {angularTargetAxis=AxisZ, twistAxis=AxisZ, angularLimitsMax=<15 3 0>, angularLimitsMin=<-15 -3 0>, angularTarget=<0 0 0>}
      localJointOffset=<0 0 -0.05>
     }]
   }
   n12 = Anim Dynamics - Bone: dyn_anklet(ComponentPose=@n11, Alpha=1, ExternalForce="0,0,0") {
    bGravityOverrideInSimSpace=true
    gravityScale=0.35
    lODThreshold=1
    numSolverIterationsPostUpdate=2
    numSolverIterationsPreUpdate=8
    alphaBoolBlend {blendOption=Linear}
    boundBone {boneName=dyn_anklet}
    physicsBodyDefinitions [{
      sphereCollisionRadius=0
      boundBone {boneName=dyn_anklet}
      boxExtents=<7 7 4>
      constraintSetup {angularLimitsMax=<0 0 0>, angularLimitsMin=<0 0 0>, angularTarget=<1 0 0>, linearAxesMax=<0.2 0.2 0.5>, linearAxesMin=<-0.2 -0.2 -0.3>}
      localJointOffset=<-0.02 0 0>
     }]
   }
   Output Pose(Result=Component To Local(ComponentPose=Observe Bone: R_eye!(ComponentPose:Component Space Pose Link Structure=@n12, bAlphaBoolEnabled:bool=True, AlphaCurveName:Name=None) -> (Pose:Component Space Pose Link Structure) {is=title AnimGraphNode_ObserveBone, displaySpace=BCS_ParentBoneSpace, alphaBoolBlend {blendOption=Linear}, boneToObserve {boneName=R_eye}})) {layerGroup=None}
   AnimGraphNode_SubInput_0 = Input Pose()
   Save cached pose 'Sub Graph Cache'(Pose=@AnimGraphNode_SubInput_0) {cacheName=Sub Graph Cache}
  ]}
 EventGraph {links [], nodes []}
]
variables []
