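# Physics asset for the pet dragon mesh, giving simple physical bodies to its pelvis, neck and head.
# Four small, identically sized capsules sit on the pelvis, two neck bones and head, with no bodies for any other bones.
# Three constraints link them in a chain from pelvis to head, each allowing moderate swing and tighter twist.
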PhysicsAsset /Game/Characters/Pets/Dragon/Meshes/Dragon_Physics
bodies {
 head [{shapeType=Capsule, length=1, radius=2.5}]
 neck_01 [{shapeType=Capsule, length=1, radius=2.5, center=<0.0551 0.3463 0>}]
 neck_02 [{shapeType=Capsule, length=1, radius=2.5}]
 pelvis [{shapeType=Capsule, length=1, radius=2.5}]
}
constraints [
 {bone1Name=neck_02, bone2Name=neck_01, swing1LimitDegrees=45, swing2LimitDegrees=45, twistLimitDegrees=15}
 {bone1Name=neck_01, bone2Name=pelvis, swing1LimitDegrees=45, swing2LimitDegrees=45, twistLimitDegrees=15}
 {bone1Name=head, bone2Name=neck_02, swing1LimitDegrees=45, swing2LimitDegrees=45, twistLimitDegrees=15}
]
